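# Niagara effect for the SpireTendBeen cosmetic shoes, adding a warm glow and short blue ribbon trails at the feet.
# Glow pins four partitioned sprites to two sockets with per-sprite offsets, sized by left and right shoe-scale parameters.
# RunningTrail draws two anim-trail ribbons with scale-driven width, fading color and a system timeline controlling alpha.
# System dependencies include a running check and boolean logic, though their wiring is absent here.
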NiagaraSystem /CosmeticShoes/Assets/FORT_Shoes/Shoe_SpireTendBeen/FX/NS_Shoe_SpireTendBeen_Idle
dependencies {
 usedDataInterfaces [
  &NiagaraDataInterfaceArrayInt32
  &NiagaraDataInterfaceArrayFloat3
  &NiagaraDataInterfaceCurve
  &NiagaraDataInterfaceColorCurve
  &NiagaraDataInterfaceSocketReader
  &NiagaraDataInterfaceArrayFloat
 ]
 usedDynamicInputs [
  &MakeVectorFromLinearColorRGB
  &ColorFromCurve
  &Boolean_AndOperation
  /Game/Effects/Niagara/DynamicInputs/Logic/FN_IsRunningCheck
  &Boolean_NotOperation
  &Boolean_OrOperation
  &Divide_Float
  &SetBoolByFloatComparison
  &VectorLength
 ]
 usedModules [
  &self:SystemSpawnScript.SystemScriptSource.SystemScriptGraph.NiagaraNodeAssignment_0.SetVariables_#1
  &SystemState
  &self:SystemSpawnScript.SystemScriptSource.SystemScriptGraph.NiagaraNodeAssignment_4.SetVariables_#2
  &self:SystemSpawnScript.SystemScriptSource.SystemScriptGraph.NiagaraNodeAssignment_1.SetVariables_#3
  &Timeline
 ]
}
emitters {
 Glow {
  emitterUpdateScript [
   &EmitterState(Life Cycle Mode=::NewEnumerator1, Loop Behavior=::NewEnumerator1, Loop Duration Mode=::NewEnumerator1, UseLoopDelay=false)
   &SpawnBurst_Instantaneous(Spawn Count=4, Use Spawn Probability=false, UseLoopCountLimit=false)
  ]
  particleSpawnScript [
   &PartitionParticles(Number Of Partitions=4)
   &InitializeParticle(
    Color=<rgba 1 0.6342 0.0612 0.2>
    Sprite Size Mode=::NewEnumerator3
    Uniform Sprite Size=12
    UsePositionOffset=false
    Write Material Random=true
   )
   &CameraOffset(Camera Offset Amount=3)
  ]
  particleUpdateScript [
   &SocketLocation(
    SamplePosOffset=true
    Sampled Position Offset=&SelectVectorFromArray(
     Direct Array Index=$Particles.PartitionParticles.Partition
     Vector Selection Array=DataInterface() {propertyValues {FloatData [<5 -6 4.5>, <4.5 -6 -5.5>, <-4.5 6 5.5>, <-4.5 6 -4.5>]}}
    )
    Socket Index=&SelectIntFromArray(Direct Array Index=$Particles.PartitionParticles.Partition, Int Selection Array=DataInterface() {propertyValues {IntData [0,0,1,1]}})
    Socket Reader=$System.Socket Reader
    Socket Sampling Mode=::NewEnumerator3
    VelocitySpeedLimit=false
   )
   &ParticleState(Kill Particles When Lifetime Has Elapsed=false, Let Infinitely Lived Particles Die When Emitter Deactivates=true)
   &ScaleSpriteSize(
    Scale Sprite Size Mode=::NewEnumerator0
    Uniform Scale Factor=&MaskFloatBySpawnGroup(
     Masked Float=&MakeFloatFromVector(Vector=$User.ShoeScale_L, Channel=::NewEnumerator0)
     Passthrough Float=&MakeFloatFromVector(Vector=$User.ShoeScale_R, Channel=::NewEnumerator0)
     Spawn Group Mask=$Particles.PartitionParticles.Partition
    )
   )
  ]
  rendererClasses [&NiagaraSpriteRendererProperties]
  renderers [{rendererClass=&NiagaraSpriteRendererProperties}]
 }
 RunningTrail {
  emitterSpawnScript [&FN_RibbonIDOffsetUtility()]
  emitterUpdateScript [
   &EmitterState()
   &FN_SpawnAnimTrail(Number of Anim Trails=2)
  ]
  particleSpawnScript [
   &FN_InitializeAnimTrailBoneIndecies(Second Bone or Socket Index=0)
   &FN_InitializeAnimTrailBoneIndecies(AnimTrail Selection=1, First Bone or Socket Index=1) {module=FN_InitializeAnimTrailBoneIndecies001}
   &FN_InitializeAnimTrail(
    AnimTrailWidth=&Multiply_Float(A=35, B=&MaskFloatBySpawnGroup(Masked Float=$System.ScalePercentL, Passthrough Float=$System.ScalePercentR))
    Color=<rgba 1 1 1 0.4>
    Lifetime=0.125
   )
  ]
  particleUpdateScript [
   &SolveForcesAndVelocity(Clamp Velocity=false, Limit Acceleration=false)
   &FN_UpdateAnimTrail() {module=FN_UpdateAnimTrail001}
   &ScaleColor(
    Linear Color Curve=DataInterface() {propertyValues {
      AlphaCurve {keys [<at 0.0144 2>,<at 0.45 1>,<at 0.8489 0>]}
      BlueCurve {keys [<at 0 0.125>,<at 0.5324 0>]}
      GreenCurve {keys [<at 0 0.0128>,<at 0.5324 0>]}
      RedCurve {keys [<at 0 0.0036>,<at 0.5324 0>]}
     }}
    Scale Mode=::NewEnumerator2
   )
   &DynamicMaterialParameters(
    Index 0 Param 1=0.85
    Index 0 Param 2=0.5
    Index 0 Param 3=-3
    Module.Index0Float0Write=true
    Module.Index0Float1Write=true
    Module.Index0Float2Write=true
    Module.Index0Float3Write=true
   )
   &ScaleColor(Scale Alpha=$System.Timeline001.Float, ScaleA=true, ScaleRGB=true) {module=ScaleColor001}
  ]
  rendererClasses [&NiagaraRibbonRendererProperties]
  renderers [{rendererClass=&NiagaraRibbonRendererProperties}]
 }
}
summary {
 systemName=NS_Shoe_SpireTendBeen_Idle
 emitters [
  {emitterName=Glow, rendererClasses [&NiagaraSpriteRendererProperties]}
  {emitterName=RunningTrail, rendererClasses [&NiagaraRibbonRendererProperties]}
 ]
 userVariables [
  User.bIsFrontEnd {type=NiagaraBool}
  User.bIsFrontEndPreview {type=NiagaraBool}
  User.bIsOffPawnPreview {type=NiagaraBool}
  User.HideFootFX {type=NiagaraBool}
  User.IsSkydiving {type=NiagaraBool}
  User.RibbonIDOffset {type=NiagaraInt32}
  User.ShoeScale_L {type=Vector3f, description="", defaultValue {struct=Vector3f, value=<1 1 1>}}
  User.ShoeScale_R {type=Vector3f, description="", defaultValue {struct=Vector3f, value=<1 1 1>}}
  User.SkeletalMesh {
   type=NiagaraDataInterfaceSkeletalMesh
   defaultValue {
    struct=NiagaraExt_VariableValue_DataInterface
    value {dataInterface=&self:NiagaraDataInterfaceSkeletalMesh_0}
   }
  }
  User.TrailOn {type=NiagaraFloat, defaultValue=1, description=""}
 ]
}
