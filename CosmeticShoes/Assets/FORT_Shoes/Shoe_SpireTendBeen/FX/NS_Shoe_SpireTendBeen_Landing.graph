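# Niagara system for the SpireTendBeen cosmetic shoe's landing effect, a short-lived sprite smoke puff at the feet.
# One Smoke emitter bursts particles partitioned across skeletal mesh bones or sockets, placed on rings, with size and radius scaled by the left or right ShoeScale user parameter.
# Puffs rise and push outward under drag, tinted by a system TODColor.
# Curves grow the sprites, fade color and alpha, and animate dynamic material parameters over lifetime.
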NiagaraSystem /CosmeticShoes/Assets/FORT_Shoes/Shoe_SpireTendBeen/FX/NS_Shoe_SpireTendBeen_Landing
dependencies {
 usedDataInterfaces [
  &NiagaraDataInterfaceCurve
  &NiagaraDataInterfaceColorCurve
  &NiagaraDataInterfaceSkeletalMesh
  &NiagaraDataInterfaceArrayFloat3
 ]
 usedDynamicInputs [&Divide_Float]
 usedModules [
  &self:SystemSpawnScript.SystemScriptSource.SystemScriptGraph.NiagaraNodeAssignment_0.SetVariables_#1
  &self:SystemSpawnScript.SystemScriptSource.SystemScriptGraph.NiagaraNodeAssignment_1.SetVariables_#2
  &SystemState
  &self:SystemSpawnScript.SystemScriptSource.SystemScriptGraph.NiagaraNodeAssignment_3.SetVariables_#3
 ]
}
emitters {Smoke {
  emitterUpdateScript [
   &EmitterState()
   &SpawnBurst_Instantaneous(Spawn Count=12, Use Spawn Probability=false, UseLoopCountLimit=false)
  ]
  particleSpawnScript [
   &PartitionParticles()
   &InitializeParticle(
    Color=<rgba 0.5747 0.5209 0.5193 1>
    Lifetime Max=0.65
    Lifetime Min=0.35
    Lifetime Mode=::NewEnumerator1
    Sprite Rotation Mode=::NewEnumerator2
    Sprite Size Mode=::NewEnumerator3
    Uniform Sprite Size=&Multiply_Float(
     A=&RandomRangeFloat(Minimum=33, Maximum=44)
     B=&MaskFloatBySpawnGroup(
      Masked Float=&MakeFloatFromVector(Vector=$User.ShoeScale_L, Channel=::NewEnumerator0)
      Passthrough Float=&MakeFloatFromVector(Vector=$User.ShoeScale_R, Channel=::NewEnumerator0)
      Spawn Group Mask=$Particles.PartitionParticles.Partition
     )
    )
    UsePositionOffset=false
    Write Material Random=true
   )
   &SkeletalMeshLocation(
    Bone / Socket Index=$Particles.PartitionParticles.Partition
    Bone Sampling Mode=::NewEnumerator4
    MaskSpawnGroup=false
    Sampled Position Offset=&SelectVectorFromArray(Direct Array Index=$Particles.PartitionParticles.Partition, Vector Selection Array=$System.SKMOffset)
    Skeletal Mesh=$System.Skeletal Mesh
   )
   &ShapeLocation(
    MaskSpawnGroup=false
    Override Seed=false
    Ring / Disc Distribution Mode:ENiagaraSphereDistributionMode=::NewEnumerator1
    Ring Radius=&Multiply_Float(
     A=9.2
     B=&MaskFloatBySpawnGroup(
      Masked Float=&MakeFloatFromVector(Vector=$User.ShoeScale_L, Channel=::NewEnumerator0)
      Passthrough Float=&MakeFloatFromVector(Vector=$User.ShoeScale_R, Channel=::NewEnumerator0)
      Spawn Group Mask=$Particles.PartitionParticles.Partition
     )
    )
    Shape Primitive=::NewEnumerator4
    U Position=&ReturnNormalizedExecIndex()
    UseOwnerScale=true
   )
   &AddVelocity(Velocity=&RandomRangeVector(Minimum=<0 0 45>, Maximum=<0 0 70>))
   &AddVelocity(
    Module.Use Falloff Distance=false
    Override Seed=false
    Velocity Mode=::NewEnumerator1
    Velocity Origin=$Particles.SkeletalMeshLocation.SampledPosition
    Velocity Speed=&RandomRangeFloat(Minimum=50, Maximum=100)
   ) {module=AddVelocity001}
   &Color(Color=&MultiplyLinearColor(Linear Color=$Particles.Color, Scale Factor=&MakeVector4_FromLinearColor(LinearColor=$System.TODColor)))
  ]
  particleUpdateScript [
   &ParticleState()
   &ScaleSpriteSize(Uniform Curve Sprite Scale=DataInterface() {propertyValues {Curve {keys [
        {interpMode=RCIM_Cubic, leaveTangent=2.1375, tangentMode=RCTM_Break, time=0, value=0.25}
        {arriveTangent=0.452, tangentMode=RCTM_User, time=1, value=1.2}
       ]}}})
   &Drag(Drag=4, Use Linear Drag=true, Use Rotational Drag=true)
   &SolveForcesAndVelocity(Clamp Velocity=false, Limit Acceleration=false)
   &DynamicMaterialParameters(
    Index 0 Param 1=&FloatFromCurve(FloatCurve=DataInterface() {propertyValues {Curve {defaultValue=1, keys [{tangentMode=RCTM_User, time=0, value=0}, {tangentMode=RCTM_User, time=1, value=1}]}}})
    Index 0 Param 3=&FloatFromCurve(FloatCurve=DataInterface() {propertyValues {Curve {
        defaultValue=1
        keys [{interpMode=RCIM_Cubic, tangentMode=RCTM_User, time=0, value=0.002}, {interpMode=RCIM_Cubic, tangentMode=RCTM_User, time=0.35, value=0}]
       }}})
    Index 0 Param 4=1
    Index 2 Param 1=&FloatFromCurve(FloatCurve=DataInterface() {propertyValues {Curve {keys [{interpMode=RCIM_Cubic, leaveTangent=3.3442, tangentMode=RCTM_Break, time=0, value=-0.05}, <at 0.7 1>]}}})
    Index 2 Param 3=&FloatFromCurve(
     FloatCurve=DataInterface() {propertyValues {Curve {keys [{interpMode=RCIM_Cubic, leaveTangent=2.4944, tangentMode=RCTM_Break, time=0, value=0}, <at 0.8 1>]}}}
     Scale Curve=2
    )
    Index 2 Param 4=&FloatFromCurve(FloatCurve=DataInterface() {propertyValues {Curve {keys [<at 0 0.25>,<at 0.5 1.05>]}}}, Scale Curve=0.7)
    Module.Index0Float0Write=true
    Module.Index0Float1Write=true
    Module.Index0Float2Write=true
    Module.Index0Float3Write=true
    Module.Index2Float0Write=true
    Module.Index2Float1Write=true
    Module.Index2Float2Write=true
    Module.Index2Float3Write=true
    Write Parameter Index 2=::NewEnumerator1
   )
   &ScaleColor(
    Linear Color Curve=DataInterface() {propertyValues {
      AlphaCurve {keys [<at 0.0104 1>,<at 0.917 0>]}
      BlueCurve {keys [<at 0 0.0312>,<at 0.9498 0>]}
      GreenCurve {keys [<at 0 0.0031>,<at 0.9498 0>]}
      RedCurve {keys [<at 0 0.0003>,<at 0.9498 0>]}
     }}
    Scale Mode=::NewEnumerator2
   )
  ]
  rendererClasses [&NiagaraSpriteRendererProperties]
  renderers [{rendererClass=&NiagaraSpriteRendererProperties}]
 }}
summary {
 systemName=NS_Shoe_SpireTendBeen_Landing
 emitters [{emitterName=Smoke, rendererClasses [&NiagaraSpriteRendererProperties]}]
 userVariables [
  User.bIsFrontEnd {type=NiagaraBool}
  User.bIsFrontEndPreview {type=NiagaraBool}
  User.ShoeScale_L {type=Vector3f, description="", defaultValue {struct=Vector3f, value=<1 1 1>}}
  User.ShoeScale_R {type=Vector3f, description="", defaultValue {struct=Vector3f, value=<1 1 1>}}
 ]
}
